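# One-shot mono MetaSound source that plays the death sound of the TidalCrane Brute enemy in ProtoSwarm.
# On play, it randomly picks one of three death waves and feeds it to a wave player with pitch shift randomized around a base by a spread.
# A fixed decibel input, converted to linear gain, scales the output, and playback completion fires OnFinished.
# The source uses its own attenuation, an enemy-damage sound class and a concurrency override.
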MetaSoundSource /ProtoSwarm/Assets/Audio/TidalCrane_Brute/MSS_TidalCrane_Brute_Death
inputs [
 Random.Pitch.Base {type=Float}
 Random.Pitch.Spread {type=Float}
 "Random.Volume.Volume Base (dB)" {type=Float}
 Sounds {type=WaveAsset:Array}
 UE.Source.OnPlay {type=Trigger}
]
interfaces [UE.OutputFormat.Mono:1.0, UE.Source.OneShot:1.0, UE.Source:1.1]
outputs [
 "UE.OutputFormat.Mono.Audio:0" {type=Audio}
 UE.Source.OneShot.OnFinished {type=Trigger}
]
pages [Default {
  input_defaults {
   Random.Pitch.Base=0
   Random.Pitch.Spread=1
   "Random.Volume.Volume Base (dB)"=10
   UE.Source.OnPlay=null
   Sounds [
    /ProtoSwarm/Assets/Audio/TidalCrane_Brute/SoundWaves/TidalCrane_Brute_Death_01
    /ProtoSwarm/Assets/Audio/TidalCrane_Brute/SoundWaves/TidalCrane_Brute_Death_02
    /ProtoSwarm/Assets/Audio/TidalCrane_Brute/SoundWaves/TidalCrane_Brute_Death_03
   ]
  }
  nodes [
   DecibelstoLinearGain_0 = UE.Decibels to Linear Gain:Float(Decibels=^"Random.Volume.Volume Base (dB)")
   ArrayRandomGet_0 = Array.Random Get:WaveAsset:Array(In Array=^Sounds, Next=^UE.Source.OnPlay) {also {defaults {Weights []}}}
   WavePlayer_0 = UE.Wave Player:Mono(
    Play=@ArrayRandomGet_0.On Next
    Wave Asset=@ArrayRandomGet_0.Value
    Pitch Shift=UE.RandomFloat(
     Next=^UE.Source.OnPlay
     Max=UE.Add:Float(PrimaryOperand=^Random.Pitch.Base, AdditionalOperands=^Random.Pitch.Spread).Out
     Min=UE.Subtract:Float(PrimaryOperand=^Random.Pitch.Base, AdditionalOperands=^Random.Pitch.Spread).Out
    ).Value
   )
   Multiply_0 = UE.Multiply:Audio by Float(AdditionalOperands=@DecibelstoLinearGain_0.Linear Gain, PrimaryOperand=@WavePlayer_0.Out Mono)
  ]
  links [{dst="@UE.OutputFormat.Mono.Audio:0", src=Multiply_0.Out}, {dst="@UE.Source.OneShot.OnFinished", src=WavePlayer_0.On Finished}]
 }]
properties {
 assetImportData=&self:AssetImportData
 attenuationSettings=&/ATT_TidalCrane_Brute_Death
 bOverrideConcurrency=true
 maxDistance=20217.5
 sampleRate=48000
 soundClassObject=/Game/Sounds/FortSoundClassesAndModes/Fort_Enemy_Damaged_PSM
 volume=2
 concurrencyOverrides {maxCount=3}
}
